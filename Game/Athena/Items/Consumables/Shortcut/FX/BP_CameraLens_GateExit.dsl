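; BP_CameraLens_GateExit is a camera lens effect Blueprint for the Shortcut consumable, shown by its name to play a gate exit crack effect.
; On BeginPlay it forces its NS_CameraLens_CrackExit Niagara component into solo mode, speeds it up through custom time dilation, then activates it with a reset.
; Tick and the construction script contain no logic.
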(event EventBeginPlay
  (bind _ns_cameralens_crackexit (Variables|Default|GetNS_CameraLens_CrackExit))
  (Niagara|SetForcedSoloMode _ns_cameralens_crackexit true)
  (Effects|SetCustomTimeDilation _ns_cameralens_crackexit 5.0)
  (Components|Activation|Activate (Variables|Default|GetNS_CameraLens_CrackExit) true))

(event EventTick
  (params (DeltaSeconds :type Float%20%28single-precision%29)))

(fn ConstructionScript)
